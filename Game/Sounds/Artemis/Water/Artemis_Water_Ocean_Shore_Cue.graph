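# Ambient sound cue for an ocean shore, layering looping wave beds by listener distance and routed to the Fort_Water_Ambient sound class.
# Four looping waves (small shore waves, close, medium-distant and distant) each pass through their own very close, close, medium or distant attenuation asset.
# A mixer sums the layers with per-layer volumes, weighting the very close layer highest, so the per-branch attenuations shape the blend over distance.
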SoundCue /Game/Sounds/Artemis/Water/Artemis_Water_Ocean_Shore_Cue
root=Mixer_2
nodes [
 {
  id=Attenuation_6
  children [WavePlayer_7]
  properties {attenuationSettings=/Game/Athena/Sounds/Water/Attenuations/Apollo_Water_Ocean_Shore_Close_Att}
 }
 {
  id=Attenuation_7
  children [WavePlayer_6]
  properties {attenuationSettings=/Game/Athena/Sounds/Water/Attenuations/Apollo_Water_Ocean_Shore_Medium_Att}
 }
 {
  id=Attenuation_8
  children [WavePlayer_3]
  properties {attenuationSettings=/Game/Athena/Sounds/Water/Attenuations/Apollo_Water_Ocean_Shore_Distant_Att}
 }
 {
  id=Attenuation_9
  children [WavePlayer_8]
  properties {attenuationSettings=/Game/Athena/Sounds/Water/Attenuations/Apollo_Water_Ocean_Shore_Very_Close_Att}
 }
 {id=Mixer_2, children [Attenuation_9, Attenuation_6, Attenuation_7, Attenuation_8], properties {inputVolume [1,0.75,0.5,0.75]}}
 {id=WavePlayer_3, properties {bLooping=true, soundWaveAssetPtr=/Game/Athena/Sounds/Water/Ambient/Amb_Water_Ocean_Waves_Distant_Loop_01}}
 {id=WavePlayer_6, properties {bLooping=true, soundWaveAssetPtr=/Game/Athena/Sounds/Water/Ambient/Amb_Water_Ocean_Waves_Med_Distant_Loop_01}}
 {id=WavePlayer_7, properties {bLooping=true, soundWaveAssetPtr=/Game/Athena/Sounds/Water/Ambient/Amb_Water_Ocean_Waves_Close_Loop_01}}
 {id=WavePlayer_8, properties {bLooping=true, soundWaveAssetPtr=/Game/Athena/Sounds/Water/Ambient/Amb_Water_Ocean_Shore_SmWaves_Loop_01}}
]
properties {
 maxDistance=10000
 soundClassObject=/Game/Sounds/FortSoundClassesAndModes/Fort_Water_Ambient
 volumeMultiplier=1.5
 attenuationOverrides {
  bAttenuateWithLPF=true
  falloffDistance=6000
  lPFFrequencyAtMax=500
  nonSpatializedRadiusStart=700
  stereoSpread=1000
  attenuationShapeExtents {x=925}
 }
 concurrencyOverrides {maxCount=1, resolutionRule=PreventNew}
}
