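# Looping ambient river water cue in the water ambient sound class, layered by listener distance.
# A mixer combines close 2D, close attenuated and medium layers, each with its own attenuation asset.
# In each layer, a Velocity parameter brings in a louder velocity-blend loop over a detail loop, which fades out only in the close layers as Velocity nears one.
# An Intensity parameter modulates overall volume at the root.
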SoundCue /Game/Sounds/Artemis/Water/Artemis_Water_River_Default_Cue
root=ModulatorContinuous_0
nodes [
 {
  id=Attenuation_0
  children [ParamCrossFade_0]
  properties {attenuationSettings=/Game/Athena/Sounds/Water/Attenuations/Apollo_Water_River_Close_2D_Att}
 }
 {
  id=Attenuation_1
  children [ParamCrossFade_1]
  properties {attenuationSettings=/Game/Athena/Sounds/Water/Attenuations/Apollo_Water_River_Close_Att}
 }
 {
  id=Attenuation_2
  children [ParamCrossFade_2]
  properties {attenuationSettings=/Game/Athena/Sounds/Water/Attenuations/Apollo_Water_River_Medium_Att}
 }
 {id=Mixer_0, children [Attenuation_0, Attenuation_1, Attenuation_2], properties {inputVolume [1,1,1]}}
 {id=ModulatorContinuous_0, children [Mixer_0], properties {volumeModulationParams {parameterName=Intensity}}}
 {
  id=ParamCrossFade_0
  children [WavePlayer_5,WavePlayer_6]
  properties {
   paramName=Velocity
   crossFadeInput [
    {fadeInDistanceEnd=0, fadeInDistanceStart=0, fadeOutDistanceEnd=1, fadeOutDistanceStart=0.75, volume=1}
    {fadeInDistanceEnd=1, fadeInDistanceStart=0.75, fadeOutDistanceEnd=100, fadeOutDistanceStart=100, volume=2.5}
   ]
  }
 }
 {
  id=ParamCrossFade_1
  children [WavePlayer_7,WavePlayer_8]
  properties {
   paramName=Velocity
   crossFadeInput [
    {fadeInDistanceEnd=0, fadeInDistanceStart=0, fadeOutDistanceEnd=1, fadeOutDistanceStart=0.75, volume=1}
    {fadeInDistanceEnd=1, fadeInDistanceStart=0.75, fadeOutDistanceEnd=100, fadeOutDistanceStart=100, volume=2.5}
   ]
  }
 }
 {
  id=ParamCrossFade_2
  children [WavePlayer_12, WavePlayer_9]
  properties {
   paramName=Velocity
   crossFadeInput [
    {fadeInDistanceEnd=0, fadeInDistanceStart=0, fadeOutDistanceEnd=100, fadeOutDistanceStart=1, volume=1}
    {fadeInDistanceEnd=1, fadeInDistanceStart=0.75, fadeOutDistanceEnd=100, fadeOutDistanceStart=100, volume=2.5}
   ]
  }
 }
 {id=WavePlayer_12, properties {bLooping=true, soundWaveAssetPtr=&/Artemis_Water_River_Detail_Medium_Loop_Att_01}}
 {id=WavePlayer_5, properties {bLooping=true, soundWaveAssetPtr=&/Artemis_Water_River_Detail_Close_Loop_2D_01}}
 {
  id=WavePlayer_6
  properties {bLooping=true, soundWaveAssetPtr=&/Artemis_Water_River_Velocity_Blend_Close_Loop_2D_01}
 }
 {id=WavePlayer_7, properties {bLooping=true, soundWaveAssetPtr=&/Artemis_Water_River_Detail_Close_Loop_Att_01}}
 {
  id=WavePlayer_8
  properties {bLooping=true, soundWaveAssetPtr=&/Artemis_Water_River_Velocity_Blend_Close_Loop_Att_01}
 }
 {
  id=WavePlayer_9
  properties {bLooping=true, soundWaveAssetPtr=&/Artemis_Water_River_Velocity_Blend_Medium_Loop_Att_01}
 }
]
properties {
 maxDistance=6000
 soundClassObject=/Game/Sounds/FortSoundClassesAndModes/Fort_Water_Ambient
 volumeMultiplier=1
 attenuationOverrides {
  bAttenuateWithLPF=true
  bEnableListenerFocus=true
  falloffDistance=6000
  lPFFrequencyAtMax=500
  nonSpatializedRadiusStart=700
  stereoSpread=1000
  attenuationShapeExtents {x=925}
 }
 concurrencyOverrides {maxCount=1, resolutionRule=PreventNew}
}
